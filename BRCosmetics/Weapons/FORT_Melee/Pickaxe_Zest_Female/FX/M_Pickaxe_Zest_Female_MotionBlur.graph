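# Unlit translucent, two-sided Niagara mesh-particle material for the Zest Female pickaxe's swing motion blur.
# A panned motion-blur texture, multiplied by a curve-atlas falloff along UV, masks both subtle refraction and Fresnel-weighted, depth-faded opacity.
# Emissive blends two colors by Fresnel, adds panning noise streaks and passes through an emissive-adjustment function, with particle color tinting it and particle alpha scaling opacity.
# A second panning noise pushes vertices along their normals.
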Material /BRCosmetics/Weapons/FORT_Melee/Pickaxe_Zest_Female/FX/M_Pickaxe_Zest_Female_MotionBlur
graph {
 properties {
  bUsedWithNiagaraMeshParticles=true
  blendMode=BLEND_Translucent
  pixelDepthOffsetMode=PDOM_Legacy
  refractionMethod=RM_IndexOfRefraction
  shadingModel=MSM_Unlit
  twoSided=true
  usedShadingModels=Unlit
 }
 nodes [
  n1 = MaterialFunctionCall(XY=TextureCoordinate()) {materialFunction=&Swizzle}
  n2 = VectorParameter() {parameterName=Mask UV, defaultValue=<rgba 1 2 0 1>}
  n3 = AppendVector(A=@n2.R, B=@n2.R)
  n4 = TextureSample(UVs=Panner(Coordinate=Multiply(A=@n1.YX, B=@n3), Speed=AppendVector(A=@n2.B, B=@n2.B))) {samplerSource=SSM_Wrap_WorldGroupSettings, texture=&/T_Pickaxe_Zest_Female_MotionBlur}
  n5 = Saturate(None=LinearInterpolate(
    A=ScalarParameter() {defaultValue=0.05, parameterName=Mask Dark}
    B=ScalarParameter() {defaultValue=1, parameterName=Mask Light}
    Alpha=@n4.G
   ) {constA=0.1})
  n6 = CurveAtlasRowParameter(CurveTime=ComponentMask(None=TextureCoordinate()) {g=true}) {
   atlas=/Game/Athena/LinearColorCurves/VFX_CurveAtlas
   curve=/Game/Weapons/FORT_Melee/Pickaxe_Zest_Female/FX/Curve_Pickaxe_Zest_Female
   defaultValue=61
  }
  n7 = Power(
   Base=ChannelMaskParameter(None=AppendVector(A=@n6, B=@n6)) {maskChannel=Alpha, parameterName=Curve Channel, defaultValue=<rgba 0 0 0 1>}.R
   Exp=ScalarParameter() {defaultValue=1, parameterName=Curve Power}
  )
  LinearInterpolate(Alpha=Multiply(A=ScalarParameter() {defaultValue=0.5, parameterName=Refraction Intensity}, B=Multiply(A=@n5, B=@n7))) {constA=1, constB=1.05}
  n8 = VectorParameter() {parameterName=Line UV, defaultValue=<rgba 1 10 8 0>}
  n9 = Multiply(A=TextureCoordinate(), B=AppendVector(A=@n8.R, B=@n8.R))
  n10 = AppendVector(A=@n8.B, B=@n8.B)
  n11 = LinearInterpolate(
   A=TextureSample(UVs=Panner(Coordinate=@n9, Speed=@n10)) {samplerType=SAMPLERTYPE_Grayscale, texture=&LowResBlurredNoise}.R
   Alpha=ScalarParameter() {defaultValue=0.2, parameterName=Ues Line}
  ) {constAlpha=1}
  n12 = Fresnel(
   ExponentIn=ScalarParameter() {defaultValue=3, parameterName=Fresnel Exp}
   BaseReflectFractionIn=ScalarParameter() {defaultValue=0.01, parameterName=Fresnel Base}
  ) {baseReflectFraction=0.005}
  n13 = LinearInterpolate(
   A=VectorParameter() {parameterName=Color A, defaultValue=<rgba 1 0.2 1 1>}.RGB
   B=VectorParameter() {parameterName=Color B, defaultValue=<rgba 0 0 10 1>}.RGB
   Alpha=@n12
  )
  n14 = ParticleColor()
  MaterialFunctionCall(Emissive=Multiply(A=Multiply(A=@n11, B=@n13), B=@n14.RGB)) {materialFunction=&MF_FX_EmissiveAdjustments}
  n15 = VectorParameter() {parameterName=WPO UV, defaultValue=<rgba 1 1 10 0>}
  n16 = Panner(Coordinate=Multiply(A=TextureCoordinate(), B=AppendVector(A=@n15.R, B=@n15.R)), Speed=AppendVector(A=@n15.B, B=@n15.B))
  Multiply(
   A=Multiply(
    A=TextureSample(UVs=@n16) {samplerType=SAMPLERTYPE_Grayscale, texture=&LowResBlurredNoise}.R
    B=VertexNormalWS()
   )
   B=ScalarParameter() {defaultValue=2, parameterName=WPO Intensity}
  ) {constB=2}
  n17 = Multiply(A=Multiply(A=@n12, B=@n5), B=@n7)
  n18 = ScalarParameter() {defaultValue=40, parameterName=Opacity Mul}
  Multiply(A=@n14.A, B=DepthFade(Opacity=Multiply(A=@n17, B=@n18), FadeDistance=ScalarParameter() {defaultValue=20, parameterName=FadeDistance}))
 ]
}
