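# Material function for the Zest Female pickaxe that paints a glowing LED light at a parameterized, offsettable point on the mesh.
# Three sphere masks on pre-skinned local position form a view-facing core, a pulsing middle halo limited to Epic quality, and a wide outer glow.
# The core mask blends two input colors.
# A scalar input scales the summed glow, falling back below one to a dim default that the Hit input fades out.
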MaterialFunction /BRCosmetics/Weapons/FORT_Melee/Pickaxe_Zest_Female/Materials/MF_positionLight
graph {nodes [
  Constant()
  Constant() {r=1}
  Constant()
  n1 = FunctionInput(Preview=Constant() {r=1}) {bUsePreviewValueAsDefault=true, inputType=FunctionInput_Scalar}
  n2 = Constant() {r=1}
  n3 = LinearInterpolate(
   A=ScalarParameter() {defaultValue=0.08, parameterName=LED_defult_brightness}
   B=Constant()
   Alpha=FunctionInput(Preview=Constant()) {bUsePreviewValueAsDefault=true, inputName=Hit, inputType=FunctionInput_Scalar}
  )
  n4 = Constant() {r=-0.08}
  n5 = QualitySwitch(Default=@n4, Low=@n4, High=@n4, Medium=@n4, Epic=ScalarParameter() {defaultValue=-0.1079, parameterName=LED_width})
  n6 = Constant() {r=1}
  n7 = OneMinus(None=Fresnel(ExponentIn=Constant() {r=0.0934}))
  n8 = PreSkinnedPosition()
  n9 = FunctionInput(Preview=ShaderStageSwitch(PixelShader=VertexInterpolator(VS=@n8).PS, VertexShader=@n8)) {bUsePreviewValueAsDefault=true, inputName=LocalPosition}
  n10 = Constant()
  n11 = VectorParameter() {parameterName=LEd_location, defaultValue=<rgba -0.036 0 0 1>}
  n12 = Constant() {r=12}
  n13 = Constant()
  n14 = MaterialFunctionCall(R=@n11.RGB, G=@n11.RGB, B=Add(A=Add(A=@n12, B=@n11.RGB), B=FunctionInput(Preview=@n13) {inputName=PositionOffset})) {materialFunction=&AppendMany}
  n15 = Constant() {r=2.4}
  n16 = SphereMask(
   A=@n9
   B=Add(A=@n10, B=@n14.RGB)
   Radius=QualitySwitch(Default=@n15, Low=@n15, High=@n15, Medium=@n15, Epic=ScalarParameter() {defaultValue=2.25, parameterName=LED_Radius})
   Hardness=ScalarParameter() {defaultValue=-0.5758, parameterName=LED_Hardness}
  )
  n17 = ScalarParameter() {defaultValue=2, parameterName=LED_power}
  n18 = Multiply(
   A=Saturate(None=LinearInterpolate(A=@n5, B=@n6, Alpha=@n7))
   B=Multiply(A=Power(Base=@n16, Exp=@n17), B=ScalarParameter() {defaultValue=1000, parameterName=LED_brightness})
  )
  n19 = LinearInterpolate(
   A=FunctionInput(Preview=Constant3Vector() {constant=<rgba 0.652 0.1249 0.6939 1>}) {inputName=outer}
   B=FunctionInput(Preview=Constant3Vector() {constant=<rgba 1.239 0.0967 1.5 1>}) {inputName=inter1}
   Alpha=Saturate(None=@n18)
  )
  n20 = Constant()
  n21 = Fresnel(ExponentIn=Constant() {r=0.2416})
  n22 = Power(
   Base=SphereMask(
    A=@n9
    B=Add(A=@n10, B=@n14.RGB)
    Radius=ScalarParameter() {defaultValue=3.6278, parameterName=Led_second_Radius}
    Hardness=ScalarParameter() {defaultValue=-0.36, parameterName=Led_second_hardness}
   )
   Exp=ScalarParameter() {defaultValue=2, parameterName=Led_second__power}
  )
  n23 = ScalarParameter() {defaultValue=60, parameterName=Led_second__brightness}
  n24 = Multiply(A=Time()) {constB=30}
  n25 = Power(Base=ConstantBiasScale(None=Sine(None=@n24) {period=8}) {bias=0.5, scale=1}, Exp=Constant() {r=2})
  n26 = QualitySwitch(Default=@n23, Low=@n23, High=@n23, Medium=@n23, Epic=LinearInterpolate(A=@n23, B=Subtract(A=@n23) {constB=15}, Alpha=Clamp(None=@n25)))
  n27 = QualitySwitch(Default=@n20, Low=@n20, High=@n20, Medium=@n20, Epic=Multiply(A=OneMinus(None=@n21), B=Multiply(A=@n22, B=@n26)))
  n28 = Add(A=@n10, B=@n14.RGB)
  n29 = ScalarParameter() {defaultValue=11, parameterName=Led_third_Radius}
  n30 = ScalarParameter() {defaultValue=-0.1298, parameterName=Led_third_Hardness}
  n31 = ScalarParameter() {defaultValue=3, parameterName=Led_third_brightness}
  n32 = Multiply(
   A=Power(Base=SphereMask(A=@n9, B=@n28, Radius=@n29, Hardness=@n30), Exp=ScalarParameter() {defaultValue=2, parameterName=Led_third_power})
   B=@n31
  )
  n33 = Multiply(A=Add(A=Add(A=@n18, B=@n27), B=@n32), B=ScalarParameter() {defaultValue=0.45, parameterName=Global_brightness})
  FunctionOutput(None=Multiply(A=If(A=@n1, B=@n2, "A > B"=@n1, "A == B"=@n1, "A < B"=@n3), B=Multiply(A=@n19, B=@n33)))
  n34 = Constant()
  QualitySwitch(Default=@n34, Low=@n34, High=@n34, Medium=@n34, Epic=@n32)
  QualitySwitch(Default=@n31, Low=@n31, High=@n31, Medium=@n31, Epic=Constant() {r=3})
 ]}
