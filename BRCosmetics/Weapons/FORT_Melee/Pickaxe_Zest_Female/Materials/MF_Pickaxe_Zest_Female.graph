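# Material function for the female Zest pickaxe that darkens masked base color and builds an animated emissive reacting to hits and event audio.
# Five colored, position-offset light bands switch on at stepped thresholds.
# A pickaxe parameter collection drives hit count, hit state and hue shift, alongside Fortnite's audio spectrum.
# Emissive mask channels blend the bands, hue-shifted layers and a sphere-mask glow, with richer animation only at Epic quality.
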MaterialFunction /BRCosmetics/Weapons/FORT_Melee/Pickaxe_Zest_Female/Materials/MF_Pickaxe_Zest_Female
graph {nodes [
  n1 = NamedRerouteUsage()
  Composite!(None=@n1)
  n2 = NamedRerouteUsage()
  Composite(None=@n2)
  n3 = NamedRerouteUsage()
  Composite(None=@n3)
  n4 = NamedRerouteUsage()
  Composite(None=@n4)
  n5 = NamedRerouteUsage()
  Composite(None=@n5)
  Constant()
  n6 = GetMaterialAttributes(None=FunctionInput() {inputName=Material, inputType=FunctionInput_MaterialAttributes, sortPriority=2}) {attributeGetTypes [#1, #2]}
  n7 = LinearInterpolate(A=@n6.BaseColor, B=Constant3Vector(), Alpha=ScalarParameter() {defaultValue=0.8, parameterName=Color_saturate}) {constAlpha=0.1}
  n8 = TextureSampleParameter2D() {parameterName=Emissive_Mask, texture=/BRCosmetics/Weapons/FORT_Melee/Pickaxe_Zest_Female/Textures/T_Pickaxe_Zest_E_Mask}
  FunctionOutput(None=SetMaterialAttributes!(
    MaterialAttributes=@n6.MaterialAttributes
    EmissiveColor=NamedRerouteUsage()
    BaseColor=LinearInterpolate(A=@n6.BaseColor, B=@n7, Alpha=@n8.G)
   ) {attributeSetTypes [#3, #1]})
  n9 = NamedRerouteUsage()
  n10 = CollectionParameter() {collection=&FortniteMaterialParameters, parameterName=EventAudio_Spectrum}
  n11 = MaterialFunctionCall(Float4=Power(Base=@n10) {constExponent=1}) {materialFunction=&BreakOutFloat4Components}
  n12 = MaterialFunctionCall(
   Input=MaterialFunctionCall!(1=@n11.R, 2=@n11.R, 3=@n11.R, 4=@n11.R, A) {materialFunction=/Engine/Functions/Engine_MaterialFunctions02/Lerp_Multiple_Float}.Lerp 4 Inputs
   Input Low=Constant() {r=0.2}
   Input High=Constant() {r=0.75}
  ) {materialFunction=&RemapValue_0-1}
  n13 = NamedRerouteUsage()
  n14 = CollectionParameter() {collection=&/MPC_Pickaxe_zest_female, parameterName=Hit count}
  n15 = Constant() {r=0.2}
  NamedRerouteDeclaration(None=LinearInterpolate(
    A=LinearInterpolate(A=@n9, B=@n12.Result, Alpha=@n13)
    B=Multiply(A=@n14, B=@n15)
    Alpha=CollectionParameter() {collection=&MPC_Pickaxe_zest_female, parameterName=Hiting?}
   )) {name=TubeLight, nodeColor=<rgba 1 0.649 0 1>}
  n16 = CollectionParameter() {collection=&FortniteMaterialParameters, parameterName=EventAudio_Spectrum}
  n17 = Max(
   A=CollectionParameter() {collection=&FortniteMaterialParameters, parameterName=Emote_EnvTime}
   B=MaterialFunctionCall(Float4=Power(Base=@n16)) {materialFunction=&BreakOutFloat4Components}.R
  )
  n18 = Constant()
  NamedRerouteDeclaration(None=If(A=Saturate(None=@n17), B=Constant(), "A > B"=Constant() {r=1}, "A == B"=@n18, "A < B"=@n18)) {name=audio, nodeColor=<rgba 1 0 0.0469 1>}
  n19 = ConstantBiasScale(None=Sine(None=Time()) {period=4}) {bias=0.5, scale=1}
  n20 = ScalarParameter() {defaultValue=1, parameterName=Param_4}
  NamedRerouteDeclaration(None=Clamp(None=Power(Base=@n19, Exp=@n20))) {name=TimeAnimation, nodeColor=<rgba 0.6875 0 1 1>}
  NamedRerouteDeclaration(None=TextureSampleParameter2D() {parameterName=Param, texture=/BRCosmetics/Weapons/FORT_Melee/Pickaxe_Zest_Female/Textures/T_Pickaxe_Zest_Female_E}.RGB) {name=E_texture, nodeColor=<rgba 0 1 0.6953 1>}
  n21 = PreSkinnedPosition()
  NamedRerouteDeclaration(None=ShaderStageSwitch(PixelShader=VertexInterpolator(VS=@n21).PS, VertexShader=@n21)) {name=LocaPosition, nodeColor=<rgba 0.5469 0 1 1>}
  n22 = Multiply(
   A=CollectionParameter() {collection=&MPC_Pickaxe_zest_female, parameterName=Hit animation}
   B=Constant() {r=2}
  )
  n23 = Constant3Vector() {constant=<rgba 0.7305 0.0356 0.6376 1>}
  n24 = Constant3Vector() {constant=<rgba 1 0.1232 0.4714 1>}
  MaterialExpressionReroute_1 = Reroute(None=@n1)
  n25 = Constant()
  n26 = If(A=Saturate(None=@MaterialExpressionReroute_1.Input), B=Constant(), "A > B"=Constant() {r=1}, "A == B"=@n25, "A < B"=@n25)
  MaterialExpressionReroute_4 = Reroute(None=@n26)
  n27 = NamedRerouteUsage()
  n28 = LinearInterpolate(
   A=LinearInterpolate(Alpha=CollectionParameter() {collection=&MPC_Pickaxe_zest_female, parameterName=Hiting?})
   Alpha=NamedRerouteUsage()
  )
  n29 = Constant()
  MaterialExpressionReroute_8 = Reroute(None=@n2)
  n30 = Constant()
  n31 = If(A=Saturate(None=@MaterialExpressionReroute_8.Input), B=Constant() {r=0.2}, "A > B"=Constant() {r=1}, "A == B"=@n30, "A < B"=@n30)
  MaterialExpressionReroute_9 = Reroute(None=@n31)
  n32 = MaterialFunctionCall!(
   outer=Constant3Vector() {constant=<rgba 1.1 0.2508 0.2289 1>}
   inter1=Constant3Vector() {constant=<rgba 1 0.3841 0.3642 1>}
   In=@MaterialExpressionReroute_9.Output
   LocalPosition=@n27
   Hit=@n28
   PositionOffset=Constant() {r=6}
  ) {materialFunction=&/MF_positionLight}
  n33 = Constant3Vector() {constant=<rgba 0.2858 0.0255 1.2 1>}
  n34 = Constant3Vector() {constant=<rgba 0.6469 0.2769 1.3 1>}
  MaterialExpressionReroute_10 = Reroute(None=@n3)
  n35 = Constant()
  n36 = If(A=Saturate(None=@MaterialExpressionReroute_10.Input), B=Constant() {r=0.4}, "A > B"=Constant() {r=1}, "A == B"=@n35, "A < B"=@n35)
  MaterialExpressionReroute_11 = Reroute(None=@n36)
  n37 = Constant() {r=12}
  n38 = Constant3Vector() {constant=<rgba 0 0.4245 1 1>}
  n39 = Constant3Vector() {constant=<rgba 0 0.8015 1 1>}
  MaterialExpressionReroute_12 = Reroute(None=@n4)
  n40 = Constant()
  n41 = If(A=Saturate(None=@MaterialExpressionReroute_12.Input), B=Constant() {r=0.6}, "A > B"=Constant() {r=1}, "A == B"=@n40, "A < B"=@n40)
  MaterialExpressionReroute_13 = Reroute(None=@n41)
  n42 = Constant() {r=18}
  MaterialExpressionReroute_14 = Reroute(None=@n5)
  n43 = Constant()
  n44 = If(A=Saturate(None=@MaterialExpressionReroute_14.Input), B=Constant() {r=0.8}, "A > B"=Constant() {r=1}, "A == B"=@n43, "A < B"=@n43)
  MaterialExpressionReroute_15 = Reroute(None=@n44)
  n45 = Max(
   A=Max(
    A=MaterialFunctionCall(outer=@n33, inter1=@n34, In=@MaterialExpressionReroute_11.Output, LocalPosition=@n27, Hit=@n28, PositionOffset=@n37) {materialFunction=&MF_positionLight}.Result
    B=MaterialFunctionCall(outer=@n38, inter1=@n39, In=@MaterialExpressionReroute_13.Output, LocalPosition=@n27, Hit=@n28, PositionOffset=@n42) {materialFunction=&MF_positionLight}.Result
   )
   B=MaterialFunctionCall(
    outer=Constant3Vector() {constant=<rgba 0.6724 0.0262 0.5776 1>}
    inter1=Constant3Vector() {constant=<rgba 1 0.2747 0.8388 1>}
    In=@MaterialExpressionReroute_15.Output
    LocalPosition=@n27
    Hit=@n28
    PositionOffset=Constant() {r=24}
   ) {materialFunction=&MF_positionLight}.Result
  )
  n46 = Add(
   A=Max(
    A=MaterialFunctionCall(outer=@n23, inter1=@n24, In=@MaterialExpressionReroute_4.Output, LocalPosition=@n27, Hit=@n28, PositionOffset=@n29) {materialFunction=&MF_positionLight}.Result
    B=Max(A=@n32.Result, B=@n45)
   )
   B=Multiply(A=TextureSampleParameter2D() {parameterName=Param, texture=&T_Pickaxe_Zest_Female_E}.RGB) {constB=0.1}
  )
  n47 = Constant() {r=1}
  n48 = NamedRerouteUsage()
  n49 = ComponentMask(None=Constant3Vector() {constant=<rgba 0 0.5 0 1>}) {r=true}
  n50 = ComponentMask(None=Constant3Vector() {constant=<rgba 0 0 -0.5 1>}) {b=true}
  n51 = Frac(None=Divide(A=@n48, B=Constant() {r=8}) {constB=2.1111})
  n52 = Frac(None=Multiply(A=Time(), B=LinearInterpolate(A=@n49, B=@n50, Alpha=@n51)))
  n53 = LinearInterpolate(
   A=LinearInterpolate(A=LinearInterpolate(A=@n48, Alpha=@n52), B=Constant() {r=1}, Alpha=NamedRerouteUsage())
   B=Constant() {r=1}
   Alpha=CollectionParameter() {collection=&MPC_Pickaxe_zest_female, parameterName=Hiting?}
  )
  n54 = Saturate(None=@n53)
  n55 = If(A=@n54, B=Constant() {r=0.2}, "A > B"=@n54, "A == B"=@n54, "A < B"=Constant() {r=1})
  n56 = TextureSampleParameter2D() {parameterName=Emissive_Mask, texture=&T_Pickaxe_Zest_E_Mask}
  n57 = Multiply(
   A=Multiply(
    A=MaterialFunctionCall(Hue Shift Percentage=@n22, Texture=@n46) {materialFunction=&HueShift}.Result
    B=QualitySwitch(Default=@n47, Low=@n47, High=@n47, Medium=@n47, Epic=@n55)
   )
   B=@n56.G
  )
  n58 = Saturate(None=Add(
    A=NamedRerouteUsage()
    B=CollectionParameter() {collection=&MPC_Pickaxe_zest_female, parameterName=Hit count}
   ))
  n59 = Constant()
  n60 = AppendVector(A=@n59, B=@n59)
  n61 = Constant() {r=6}
  n62 = Constant() {r=120}
  n63 = Multiply(
   A=Divide(A=@n62, B=Constant() {r=11}) {constB=8}
   B=CollectionParameter() {collection=&MPC_Pickaxe_zest_female, parameterName=Hit count}
  )
  n64 = CollectionParameter() {collection=&MPC_Pickaxe_zest_female, parameterName=Hiting?}
  n65 = CollectionParameter() {collection=&FortniteMaterialParameters, parameterName=EventAudio_Spectrum}
  n66 = MaterialFunctionCall(Float4=Power(Base=@n65)) {materialFunction=&BreakOutFloat4Components}
  n67 = Saturate(None=MaterialFunctionCall(1=@n66.R, 2=@n66.R, 3=@n66.R, 4=@n66.R) {materialFunction=&Lerp_Multiple_Float}.Lerp 4 Inputs)
  n68 = LinearInterpolate(A=@n48, B=@n67, Alpha=NamedRerouteUsage())
  n69 = Constant() {r=1}
  n70 = CollectionParameter() {collection=&MPC_Pickaxe_zest_female, parameterName=Hiting?}
  n71 = SphereMask(
   A=NamedRerouteUsage()
   B=LinearInterpolate(
    A=@n59
    B=AppendVector(A=@n60, B=@n61)
    Alpha=CollectionParameter() {collection=&MPC_Pickaxe_zest_female, parameterName=Hiting?}
   )
   Radius=LinearInterpolate(A=Constant(), B=LinearInterpolate(A=@n62, B=@n63, Alpha=@n64), Alpha=LinearInterpolate(A=@n68, B=@n69, Alpha=@n70))
   Hardness=Constant() {r=0.5}
  )
  n72 = Multiply(
   A=MaterialFunctionCall(Hue Shift Percentage=@n22, Texture=NamedRerouteUsage()) {materialFunction=&HueShift}.Result
   B=QualitySwitch(Default=@n58, Low=@n58, High=@n58, Medium=@n58, Epic=Multiply(A=@n53, B=@n71))
  )
  n73 = Add(A=Multiply(A=@n67) {constB=3}, B=LinearInterpolate(Alpha=@n48) {constA=1, constB=0})
  n74 = Multiply(A=NamedRerouteUsage()) {constB=0.7}
  n75 = NamedRerouteUsage()
  NamedRerouteDeclaration(None=LinearInterpolate(
    A=LinearInterpolate(A=@n57, B=@n72, Alpha=@n56.R)
    B=MaterialFunctionCall(Hue Shift Percentage=@n73, Texture=@n74) {materialFunction=&HueShift}.Result
    Alpha=LinearInterpolate(A=@n56.A, B=@n56.A, Alpha=@n75)
   )) {name=Final E, nodeColor=<rgba 0 0.7891 1 1>}
  PinBase!(Output=@n26) {reroutePins [Output {expression=&self:MaterialExpressionReroute_4}]}
  PinBase!() {reroutePins [Input {expression=&self:MaterialExpressionReroute_1}]}
  PinBase() {reroutePins [Input {expression=&self:MaterialExpressionReroute_8}]}
  PinBase!(Output=@n31) {reroutePins [Output {expression=&self:MaterialExpressionReroute_9}]}
  PinBase!() {reroutePins [Input {expression=&self:MaterialExpressionReroute_10}]}
  PinBase!(Output=@n36) {reroutePins [Output {expression=&self:MaterialExpressionReroute_11}]}
  PinBase!() {reroutePins [Input {expression=&self:MaterialExpressionReroute_12}]}
  PinBase!(Output=@n41) {reroutePins [Output {expression=&self:MaterialExpressionReroute_13}]}
  PinBase!() {reroutePins [Input {expression=&self:MaterialExpressionReroute_14}]}
  PinBase!(Output=@n44) {reroutePins [Output {expression=&self:MaterialExpressionReroute_15}]}
  TextureSampleParameter2D() {parameterName=Param, texture=&T_Pickaxe_Zest_Female_E}
 ]}
